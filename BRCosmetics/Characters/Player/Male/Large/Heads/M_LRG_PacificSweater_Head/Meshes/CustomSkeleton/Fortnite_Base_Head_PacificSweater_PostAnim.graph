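# Post-process Anim Blueprint for the PacificSweater head cosmetic, posing its extra "Baby" deform bones relative to the character's head.
# The output chain additively rotates head, position-constrains deform_headBaby to it, then applies fixed offsets to baby spine and head bones.
# Updates compute head-direction alphas, counter-rotation and emote-montage flags.
# The connected pose chain uses none of these values; directional tilt and look-at/rotation constraint nodes sit disconnected from output.
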AnimBlueprint /BRCosmetics/Characters/Player/Male/Large/Heads/M_LRG_PacificSweater_Head/Meshes/CustomSkeleton/Fortnite_Base_Head_PacificSweater_PostAnim
target_skeleton=&/Fortnite_Base_Head_PacificSweater_Skeleton
graphs [
 AnimGraph {nodes [
   n1 = "Constraint - deform_pelvisBaby (U)"!(ComponentPose, bAlphaBoolEnabled:bool=True, ConstraintWeights_0=1) -> (Pose) {
    alphaInputType=Bool
    boneToModify {boneName=deform_pelvisBaby}
    constraintSetup [{offsetOption=None, transformType=LookAt, targetBone {boneName=head}}]
    constraintWeights [1]
   }
   "Constraint - deform_headBaby (R)"!(ComponentPose=@n1, bAlphaBoolEnabled:bool=True, ConstraintWeights_0=1) -> (Pose) {
    alphaInputType=Bool
    boneToModify {boneName=deform_headBaby}
    constraintSetup [{offsetOption=None, transformType=Rotation, targetBone {boneName=head}}]
    constraintWeights [1]
   }
   n2 = "Transform (Modify) Bone - Bone: deform_headBaby"(Alpha=Get Head Left() -> (Head Left:double), Translation="0,0,0", Rotation="0,0,40", Scale="1,1,1") {
    rotationMode=BMM_Additive
    rotationSpace=BCS_BoneSpace
    translationMode=BMM_Additive
    translationSpace=BCS_BoneSpace
    alphaScaleBiasClamp {bInterpResult=true, interpSpeedDecreasing=15, interpSpeedIncreasing=15}
    boneToModify {boneName=deform_headBaby}
    rotation {roll=40}
   }
   "Transform (Modify) Bone - Bone: deform_headBaby"(ComponentPose=@n2, Alpha="Get Head Right "() -> ("Head Right ":double), Translation="0,0,0", Rotation="0,0,-40", Scale="1,1,1") {
    rotationMode=BMM_Additive
    rotationSpace=BCS_BoneSpace
    translationMode=BMM_Additive
    translationSpace=BCS_BoneSpace
    alphaScaleBiasClamp {bInterpResult=true, interpSpeedDecreasing=15, interpSpeedIncreasing=15}
    boneToModify {boneName=deform_headBaby}
    rotation {roll=-40}
   }
   n3 = "Transform (Modify) Bone - Bone: deform_headBaby"(Alpha=Get Head Up() -> (Head Up:double), Translation="0,0,0", Rotation="0,-41,0", Scale="1,1,1") {
    rotationMode=BMM_Additive
    rotationSpace=BCS_BoneSpace
    alphaScaleBiasClamp {bInterpResult=true, interpSpeedDecreasing=15, interpSpeedIncreasing=15}
    boneToModify {boneName=deform_headBaby}
    rotation {yaw=-41}
   }
   "Transform (Modify) Bone - Bone: deform_headBaby"(ComponentPose=@n3, Alpha=Get Head Down() -> (Head Down:double), Translation="0,0,0", Rotation="0,11,0", Scale="1,1,1") {
    rotationMode=BMM_Additive
    rotationSpace=BCS_BoneSpace
    translationMode=BMM_Additive
    translationSpace=BCS_BoneSpace
    alphaScaleBiasClamp {bInterpResult=true, interpSpeedDecreasing=15, interpSpeedIncreasing=15}
    boneToModify {boneName=deform_headBaby}
    rotation {yaw=11}
   }
   n4 = "Transform (Modify) Bone - Bone: head"(ComponentPose=Local To Component(LocalPose=Input Pose()), Alpha=1, Translation="0,0,0", Rotation="0,-50,0", Scale="1,1,1") {
    rotationMode=BMM_Additive
    rotationSpace=BCS_BoneSpace
    translationMode=BMM_Additive
    alphaBoolBlend {blendInTime=0.5, blendOption=Linear, blendOutTime=0.5}
    alphaScaleBiasClamp {bInterpResult=true, interpSpeedDecreasing=5, interpSpeedIncreasing=5}
    boneToModify {boneName=head}
    rotation {yaw=-50}
   }
   n5 = "Constraint - deform_headBaby (P)"!(ComponentPose=@n4, bAlphaBoolEnabled:bool=True, ConstraintWeights_0=1) -> (Pose) {
    alphaInputType=Bool
    alphaBoolBlend {blendInTime=1, blendOutTime=1}
    boneToModify {boneName=deform_headBaby}
    constraintSetup [{targetBone {boneName=head}}]
    constraintWeights [1]
   }
   n6 = "Transform (Modify) Bone - Bone: deform_spine1Baby"!(ComponentPose=@n5, bAlphaBoolEnabled:bool=True, Translation="0,0,0", Rotation="0,0,0", Scale="1,1,1") -> (Pose) {
    alphaInputType=Bool
    rotationMode=BMM_Additive
    rotationSpace=BCS_BoneSpace
    translationMode=BMM_Additive
    translationSpace=BCS_BoneSpace
    alphaScaleBiasClamp {bInterpResult=true, interpSpeedDecreasing=15, interpSpeedIncreasing=15}
    boneToModify {boneName=deform_spine1Baby}
   }
   n7 = "Transform (Modify) Bone - Bone: deform_headBaby"!(ComponentPose=@n6, bAlphaBoolEnabled:bool=True, Translation="-2,0,0", Rotation="0,5,0", Scale="1,1,1") -> (Pose) {
    alphaInputType=Bool
    rotationMode=BMM_Additive
    rotationSpace=BCS_BoneSpace
    translationMode=BMM_Additive
    translationSpace=BCS_BoneSpace
    alphaBoolBlend {blendInTime=0.25, blendOutTime=0.75}
    alphaScaleBiasClamp {bInterpResult=true, interpSpeedDecreasing=15, interpSpeedIncreasing=15}
    boneToModify {boneName=deform_headBaby}
    rotation {yaw=5}
    translation {x=-2}
   }
   n8 = "Transform (Modify) Bone - Bone: deform_spine1Baby"(ComponentPose=@n7, bAlphaBoolEnabled=True, Translation="0,-5,0", Rotation="0,0,0", Scale="1,1,1") {
    alphaInputType=Bool
    rotationMode=BMM_Additive
    rotationSpace=BCS_BoneSpace
    translationMode=BMM_Additive
    translationSpace=BCS_BoneSpace
    alphaScaleBiasClamp {bInterpResult=true, interpSpeedDecreasing=15, interpSpeedIncreasing=15}
    boneToModify {boneName=deform_spine1Baby}
    translation {y=-5}
   }
   n9 = "Transform (Modify) Bone - Bone: deform_headBaby"(ComponentPose=@n8, bAlphaBoolEnabled=True, Translation="-3,-4,0", Rotation="0,5,0", Scale="1,1,1") {
    alphaInputType=Bool
    rotationMode=BMM_Additive
    rotationSpace=BCS_BoneSpace
    translationMode=BMM_Additive
    translationSpace=BCS_BoneSpace
    alphaBoolBlend {blendInTime=0.25, blendOutTime=0.75}
    alphaScaleBiasClamp {bInterpResult=true, interpSpeedDecreasing=15, interpSpeedIncreasing=15}
    boneToModify {boneName=deform_headBaby}
    rotation {yaw=5}
    translation=<-3 -4>
   }
   Output Pose(Result=Component To Local(ComponentPose=@n9))
  ]}
 EventGraph {nodes [
   Event BlueprintInitializeAnimation()
   > Set Skel Mesh(Skel Mesh=GetOwnerSkeletalMeshComponent())
   Event BlueprintUpdateAnimation()
   > n1 = Sequence!(execute) -> (then_0, then_1, then_2:Exec)
   n2 = GetAnimInstance(self=GetOwnerSkeletalMeshComponent())
   > Is Valid(InputObject=@n2)
   > n3 = Set WartyBrineEmote(WartyBrineEmote:bool=Montage_IsActive!(
     self:Anim Instance Object Reference=@n2
     Montage:Anim Montage Object Reference=/BRCosmetics/Animation/Game/MainPlayer/Emotes/WartyBrine/CLM/Emote_WartyBrine_CLM_M
    ) -> (ReturnValue:bool)) -> (Output_Get:bool)
   > n4 = Set PinkSpikeEmote(PinkSpikeEmote:bool=OR Boolean(A=@n3, B=Montage_IsActive(self=@n2, Montage=/BRCosmetics/Animation/Game/MainPlayer/Emotes/PinkSpike/Emote_PinkSpike_CLM_M))) -> (Output_Get:bool)
   > n5 = Set LogWoahEmote(LogWoahEmote:bool=Montage_IsActive(self=@n2, Montage=/BRCosmetics/Animation/Game/MainPlayer/Emotes/LogarithmWoah/Emote_Logarithmwhoa_CMM_M)) -> (Output_Get:bool)
   > n6 = Set LogKickEmote(LogKickEmote:bool=Montage_IsActive(self=@n2, Montage=/BRCosmetics/Animation/Game/MainPlayer/Emotes/CapsuleKick/Emote_LogarithmKick_CMM_M)) -> (Output_Get:bool)
   > n7 = Set PropEmotes(PropEmotes:bool=OR Boolean!(A=@n6, B=@n5, C:bool=@n4) -> (ReturnValue)) -> (Output_Get:bool)
   > n8 = Set GlassesPropEmote(GlassesPropEmote:bool=OR Boolean(A=@n6, B=@n5, C=@n3)) -> (Output_Get:bool)
   > Set FacialTestingEmote(FacialTestingEmote:bool=Montage_IsActive(self=@n2, Montage=/Game/ArtTools/Tech_Art/TestTools/Emotes/Emote_FacialTesting_Montage)) -> (Output_Get:bool)
   > Set PropEmote_Alpha(PropEmote_Alpha:bool=OR Boolean!(A=@n7, B=@n8) -> (ReturnValue)) -> (Output_Get:bool)
   n9 = Branch(execute=@n1.then_2, Condition=NOT Boolean(A=Get bIsPlayingEmote()))
   n10 = GetOwnerSkeletalMeshComponent()
   n11 = BreakRotator(InRot=BreakTransform(InTransform=GetDeltaTransformFromRefPose(self=@n10, BoneName=head, BaseName=neck_02)).Rotation)
   n12 = MakeRotator(
    Roll=float * float!(A=@n11, B=-1, C:double=Get X percentage() -> (X percentage:float)) -> (ReturnValue)
    Pitch=float * float(A=@n11.Pitch, B=-0.8, C=Get Y percentage() -> (Y percentage:float))
    Yaw=float * float(A=@n11.Yaw, B=-1, C=Get Z percentage() -> (Z percentage:float))
   )
   > Set head rot(head rot:Rotator=@n12) -> (Output_Get:Rotator)
   Branch(execute=@n9.else, Condition=Get IsYouBoreMe() -> (IsYouBoreMe:bool))
   > Set head rot(head rot=@n12)
   n13 = Get Skel Mesh()
   Set Head Up(
    execute=@n1.then_1
    Head Up:double=GetAlphaFromReferenceTransformDelta(SkeletalMeshComponent=@n13, BoneName=spine_05, BaseName=head, InputRange=(X=0,Y=50), TwistAxis="0,0,1")
   ) -> (Output_Get:double)
   > Set Head Down(Head Down:double=GetAlphaFromReferenceTransformDelta(SkeletalMeshComponent=@n13, BoneName=spine_05, BaseName=head, InputRange=(X=0,Y=-20), TwistAxis="0,0,1")) -> (Output_Get:double)
   n14 = NOT Boolean(A=Get bIsPlayingEmote())
   > Set Head Left(Head Left:double=float * float(
     A=GetAlphaFromReferenceTransformDelta(SkeletalMeshComponent=@n13, BoneName=spine_05, BaseName=neck_01, DeltaComponentType=EulerX, InputRange=(X=-11,Y=-25), TwistAxis="0,0,1")
     B="To Float (Boolean)"(InBool=@n14)
    ) {is=title K2Node_PromotableOperator}) -> (Output_Get:double)
   n15 = NOT Boolean(A=Get bIsPlayingEmote())
   > "Set Head Right "("Head Right ":double=float * float(
     A=GetAlphaFromReferenceTransformDelta(SkeletalMeshComponent=@n13, BoneName=spine_05, BaseName=neck_01, DeltaComponentType=EulerX, InputRange=(X=-11,Y=5), TwistAxis="0,0,1")
     B="To Float (Boolean)"(InBool=@n15)
    )) -> (Output_Get:double)
   TryGetPawnOwner()
  ]}
]
variables [
 FacialTestingEmote {type=boolean}
 GlassesPropEmote {type=boolean}
 Head Down {type=number}
 Head Left {type=number}
 "Head Right " {type=number}
 head rot {type=Rotator}
 Head Up {type=number}
 IsYouBoreMe {type=boolean}
 LogKickEmote {type=boolean}
 LogWoahEmote {type=boolean}
 PinkSpikeEmote {type=boolean}
 PropEmote_Alpha {type=boolean}
 PropEmotes {type=boolean}
 Skel Mesh {type=SkeletalMeshComponent}
 WartyBrineEmote {type=boolean}
 X percentage {type=number, default=0.8}
 Y percentage {type=number, default=1}
 Z percentage {type=number, default=0.8}
]
